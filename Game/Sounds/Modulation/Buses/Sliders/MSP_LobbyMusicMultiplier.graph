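# MetaSound patch that outputs a lobby music volume multiplier which takes effect according to the frontend modulation bus.
# It scales the input frontend attenuation in decibels by the CB_IsInFrontend modulator value, then converts it to linear gain.
# That gain is multiplied by a blend between 1 and the CB_Slider_LobbyMusic value, weighted by the same frontend value.
# At a frontend value of zero, the output stays at unity.
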MetaSoundPatch /Game/Sounds/Modulation/Buses/Sliders/MSP_LobbyMusicMultiplier
inputs [FrontEndAttenuationDecibels {type=Float}]
outputs [NormalizedLobbyMusicSlider {type=Float}]
pages [Default {
  input_defaults {FrontEndAttenuationDecibels=0}
  nodes [
   n1 = UE.Reroute(Value=Modulation.GetModulatorValue() {also {defaults {Modulator=/Game/Sounds/Modulation/Buses/CB_IsInFrontend}}}.Out)
   DecibelstoLinearGain_0 = UE.Decibels to Linear Gain:Float(Decibels=UE.Multiply:Float(PrimaryOperand=^FrontEndAttenuationDecibels, AdditionalOperands=@n1.Value).Out)
   n2 = UE.Subtract:Float(AdditionalOperands=Modulation.GetModulatorValue() {also {defaults {Modulator=&/CB_Slider_LobbyMusic}}}.Out) {also {defaults {PrimaryOperand=1}}}
   Multiply_0 = UE.Multiply:Float(
    PrimaryOperand=@DecibelstoLinearGain_0.Linear Gain
    AdditionalOperands=UE.Subtract:Float(AdditionalOperands=UE.Multiply:Float(PrimaryOperand=@n1.Value, AdditionalOperands=@n2.Out).Out) {also {defaults {PrimaryOperand=1}}}.Out
   )
  ]
  links [{dst="@NormalizedLobbyMusicSlider", src=Multiply_0.Out}]
 }]
